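# Material function in the DarkAndLight skydiving contrail effect folder that scales a scalar value around an adjustable midpoint.
# It subtracts Midpoint from the input, multiplies by Scale, then adds Midpoint back, like a contrast adjustment, with no clamping.
# Unconnected inputs fall back to their previews: UV red channel for the value, 0.5 for Midpoint, and a DebugTimeSine function for Scale.
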MaterialFunction /BRCosmetics/Athena/Effects/Particles/Skydiving/NS_Contrail_DarkAndLight/Functions/MF_ValueScaleToMidpoint
graph {
 properties {description=Scales a 0-1 value towards 1, libraryCategoriesText [- TuataraVFX]}
 nodes [
  n1 = FunctionInput(Preview=MaterialFunctionCall() {materialFunction=&DebugTimeSine}.Result) {
   bUsePreviewValueAsDefault=true
   description=Amount to scale input by
   inputName=Scale
   inputType=FunctionInput_Scalar
   sortPriority=2
   previewValue=<0.25 0 0 1>
  }
  n2 = TextureCoordinate()
  n3 = FunctionInput() {
   bUsePreviewValueAsDefault=true
   description=Amount to scale input by
   inputName=Midpoint
   inputType=FunctionInput_Scalar
   sortPriority=1
   previewValue=<0.5 0 0 1>
  }
  n4 = Subtract(
   A=FunctionInput(Preview=ComponentMask(None=@n2) {r=true}) {bUsePreviewValueAsDefault=true, description=Input to be scaled, inputType=FunctionInput_Scalar}
   B=@n3
  )
  FunctionOutput(None=Add(A=Multiply(A=@n1, B=@n4), B=@n3)) {description=Scaled input, outputName=Output}
 ]
}
