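# MF_OnOff is a material function in the DarkAndLight skydiving contrail effect that outputs a looping 0-to-1 on/off envelope over time.
# It wraps a time input, defaulting to game time, with Fmod over the sum of activation, active, deactivation and inactive durations.
# A saturated linear rise is multiplied by an inverted saturated linear fall, giving a repeating trapezoid pulse that starts each cycle ramping up.
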MaterialFunction /BRCosmetics/Athena/Effects/Particles/Skydiving/NS_Contrail_DarkAndLight/Functions/MF_OnOff
graph {nodes [
  n1 = FunctionInput(Preview=Time()) {
   bUsePreviewValueAsDefault=true
   description=Custom time input for offset etc. Defaults to GameTime
   inputName=Time
   inputType=FunctionInput_Scalar
   sortPriority=4
  }
  n2 = FunctionInput() {
   bUsePreviewValueAsDefault=true
   description=How long to stay in active state
   inputName=Active Duration
   inputType=FunctionInput_Scalar
   sortPriority=1
   previewValue=<7 0 0 1>
  }
  n3 = FunctionInput() {
   bUsePreviewValueAsDefault=true
   description="How long it takes to go from 0 to 1, in seconds"
   inputName=Activation Duration
   inputType=FunctionInput_Scalar
   previewValue=<2 0 0 1>
  }
  n4 = Add(A=@n2, B=@n3)
  n5 = FunctionInput() {
   bUsePreviewValueAsDefault=true
   description=How long to stay in deactivated state
   inputName=Inactive Duration
   inputType=FunctionInput_Scalar
   sortPriority=3
   previewValue=<7 0 0 1>
  }
  n6 = FunctionInput() {
   bUsePreviewValueAsDefault=true
   description="How long it takes to go from 1 to 0, in seconds"
   inputName=Deactivation Duration
   inputType=FunctionInput_Scalar
   sortPriority=2
   previewValue=<2 0 0 1>
  }
  n7 = Fmod(A=@n1, B=Add(A=@n4, B=Add(A=@n5, B=@n6)))
  n8 = Divide(A=@n7, B=@n3)
  n9 = Saturate(None=Divide(A=Subtract(A=@n7, B=@n4), B=@n6))
  FunctionOutput(None=Multiply(A=Saturate(None=@n8), B=OneMinus(None=@n9)))
 ]}
