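# Material function for the DarkAndLight skydiving contrail that outputs a transition factor driven by time of day.
# Two saturated remaps of the global TimeOfDay parameter fade it out early and back in late, giving one at night and zero by day.
# Rounding the time selects between those ramps.
# The FrontEndToggle dynamic parameter, off by default, blends to a timed MF_OnOff cycle in place of time of day.
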MaterialFunction /BRCosmetics/Athena/Effects/Particles/Skydiving/NS_Contrail_DarkAndLight/MF_DarkAndLight_Transition
graph {nodes [
  n1 = CollectionParameter() {collection=&FortniteMaterialParameters, parameterName=TimeOfDay_0-1}
  n2 = Saturate(None=MaterialFunctionCall(Input=@n1, Input Low=Constant() {r=0.14}, Input High=Constant() {r=0.17}, Target Low=Constant() {r=1}, Target High=Constant()) {materialFunction=&RemapValueRange}.Result)
  n3 = Saturate(None=MaterialFunctionCall(Input=@n1, Input Low=Constant() {r=0.88}, Input High=Constant() {r=0.91}, Target Low=Constant(), Target High=Constant() {r=1}) {materialFunction=&RemapValueRange}.Result)
  n4 = Round(None=@n1)
  n5 = Constant() {r=1.5}
  n6 = Constant() {r=5}
  FunctionOutput(None=LinearInterpolate(
    A=LinearInterpolate(A=@n2, B=@n3, Alpha=@n4)
    B=MaterialFunctionCall!(Activation Duration=@n5, Active Duration=@n6, Deactivation Duration=@n5, Inactive Duration=@n6, Time) {materialFunction=/BRCosmetics/Athena/Effects/Particles/Skydiving/NS_Contrail_DarkAndLight/Functions/MF_OnOff}.Result
    Alpha=DynamicParameter() {defaultValue=<rgba 0 0 0 0>, paramNames ["", FrontEndToggle, FrontEndTransition, ""]}.FrontEndToggle
   ))
 ]}
