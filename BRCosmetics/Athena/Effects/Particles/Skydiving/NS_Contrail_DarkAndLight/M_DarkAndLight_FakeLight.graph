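# Unlit Modulate material for Niagara sprites that fakes a light on the scene behind them for the DarkAndLight skydiving contrail.
# With depth test off, it measures distance from the particle center to the world position behind translucency, forming a steep spherical falloff scaled by radius and stretch.
# The falloff blends from white into hue-shifted Sun or Eclipse colors, picked by a shared S-curved transition function or overridden by particle color.
# Particle Stretch also drives an outward offset from the particle center.
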Material /BRCosmetics/Athena/Effects/Particles/Skydiving/NS_Contrail_DarkAndLight/M_DarkAndLight_FakeLight
graph {
 properties {
  bDisableDepthTest=true
  bUsedWithNiagaraSprites=true
  blendMode=BLEND_Modulate
  pixelDepthOffsetMode=PDOM_Legacy
  shadingModel=MSM_Unlit
  translucencyPass=MTP_BeforeDOF
  usedShadingModels=Unlit
 }
 nodes [
  n1 = VectorParameter() {parameterName="Sun Color && Hue Shift", defaultValue=<rgba 333 16.65 16.65 0.03>}
  n2 = MaterialFunctionCall() {materialFunction=&WorldPositionBehindTranslucency}
  n3 = ParticlePositionWS()
  n4 = ParticleRadius()
  n5 = ScalarParameter() {defaultValue=1, parameterName=Particle Stretch}
  n6 = Add(A=@n5)
  n7 = Saturate(None=Divide(A=Distance(A=@n2.Result, B=@n3), B=Multiply(A=@n4, B=@n6) {constB=2}))
  n8 = MaterialFunctionCall(Hue Shift Percentage=Multiply(A=@n1.A, B=@n7), Texture=@n1.RGB) {materialFunction=&HueShift}
  n9 = VectorParameter() {parameterName="Eclipse Color && Hue Shift", defaultValue=<rgba 0 37.0108 333 0.06>}
  n10 = MaterialFunctionCall(Hue Shift Percentage=Multiply(A=@n9.A, B=@n7), Texture=@n9.RGB) {materialFunction=&HueShift}
  n11 = Saturate(None=MaterialFunctionCall!(
    In=MaterialFunctionCall() {materialFunction=&/MF_DarkAndLight_Transition}.Result
    Power=ScalarParameter() {defaultValue=9, parameterName=Transition SCurve}
   ) {materialFunction=/BRCosmetics/Athena/Effects/Particles/Skydiving/NS_Contrail_DarkAndLight/Functions/MF_SCurve_Float}.Result)
  n12 = OneMinus(None=@n7)
  n13 = Multiply(A=@n12, B=@n12)
  LinearInterpolate(
   B=LinearInterpolate(
    A=LinearInterpolate(A=@n8.Result, B=@n10.Result, Alpha=@n11)
    B=ParticleColor().RGB
    Alpha=ScalarParameter() {parameterName=UseParticleColor}
   )
   Alpha=Multiply(A=@n13, B=@n13)
  ) {constA=1}
  n14 = WorldPosition()
  Reroute(None=Multiply(A=Subtract(A=@n14.XYZ, B=@n3), B=@n5))
 ]
}
