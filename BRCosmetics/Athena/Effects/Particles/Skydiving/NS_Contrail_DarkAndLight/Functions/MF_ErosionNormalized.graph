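# Material function in the DarkAndLight skydiving contrail effect folder that erodes a gradient while retaining its 0-1 range.
# It subtracts a saturated Erosion Amount from the saturated Gradient, divides by one minus that amount, then saturates the result.
# For preview, the gradient defaults to the U texture coordinate and the erosion to an animated debug time sine.
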MaterialFunction /BRCosmetics/Athena/Effects/Particles/Skydiving/NS_Contrail_DarkAndLight/Functions/MF_ErosionNormalized
graph {
 properties {description=Erodes a gradient and remaps the result to retain the 0-1 range, libraryCategoriesText [- TuataraVFX]}
 nodes [
  n1 = ComponentMask(None=TextureCoordinate()) {r=true}
  n2 = Saturate(None=FunctionInput(Preview=MaterialFunctionCall() {materialFunction=&DebugTimeSine}.Result) {bUsePreviewValueAsDefault=true, inputName=Erosion Amount, inputType=FunctionInput_Scalar, sortPriority=1})
  n3 = Subtract(A=Saturate(None=FunctionInput(Preview=@n1) {bUsePreviewValueAsDefault=true, inputName=Gradient, inputType=FunctionInput_Scalar}), B=@n2)
  n4 = Subtract(B=@n2)
  FunctionOutput(None=Saturate(None=Divide(A=@n3, B=@n4))) {description="Erosion normalized, retaining the 0-1 range"}
 ]
}
