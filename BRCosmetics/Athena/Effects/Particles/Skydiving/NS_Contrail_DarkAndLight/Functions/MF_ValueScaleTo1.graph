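# MF_ValueScaleTo1 is a material function in the TuataraVFX library that pushes a scalar 0-1 value toward 1, stored with the DarkAndLight skydiving contrail effect.
# It saturates the input, inverts it, multiplies by Scale and inverts again, so a Scale of 1 keeps the clamped input and 0 gives 1.
# Previews default to a UV gradient and a DebugTimeSine-driven Scale.
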MaterialFunction /BRCosmetics/Athena/Effects/Particles/Skydiving/NS_Contrail_DarkAndLight/Functions/MF_ValueScaleTo1
graph {
 properties {description=Scales a 0-1 value towards 1, libraryCategoriesText [- TuataraVFX]}
 nodes [
  n1 = ComponentMask(None=TextureCoordinate()) {r=true}
  n2 = OneMinus(None=Saturate(None=FunctionInput(Preview=@n1) {bUsePreviewValueAsDefault=true, description=Input to be scaled, inputType=FunctionInput_Scalar}))
  n3 = FunctionInput(Preview=MaterialFunctionCall() {materialFunction=&DebugTimeSine}.Result) {
   bUsePreviewValueAsDefault=true
   description=Amount to scale input by
   inputName=Scale
   inputType=FunctionInput_Scalar
   sortPriority=1
   previewValue=<0.25 0 0 1>
  }
  FunctionOutput(None=OneMinus(None=Multiply(A=@n2, B=@n3))) {description=Scaled input, outputName=Output}
 ]
}
