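# Material function for the DarkAndLight skydiving contrail effect that outputs time with an offset, filed under the TuataraVFX library category.
# It adds the TimeOffset channel of a Dynamic Parameter to Time, so a particle system can shift the animation phase of materials that use it.
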MaterialFunction /BRCosmetics/Athena/Effects/Particles/Skydiving/NS_Contrail_DarkAndLight/MF_DarkAndLight_TimeOffset
graph {
 properties {description=Time with Offset, libraryCategoriesText [- TuataraVFX]}
 nodes [FunctionOutput(None=Add(A=Time(), B=DynamicParameter() {paramNames [TimeOffset, "", "", ""]}.TimeOffset)) {description=Time with Offset, outputName=Time with Offset}]
}
