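# Material function for the DarkAndLight skydiving contrail ribbon, outputting a waving, spinning rainbow beam plus a hue-shifted sun color, with alpha and a TransitionGradient.
# Sun glow fades along U, its length varying with particle direction against camera direction.
# Wave-distorted UVs form a beam whose width varies along U, remapped with a square-root curve for a rounded look.
# Skewed, time-scrolled coordinates feed the rainbow color function, whose alpha erodes the beam.
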MaterialFunction /BRCosmetics/Athena/Effects/Particles/Skydiving/NS_Contrail_DarkAndLight/MF_DarkAndLight_RainbowRibbon
graph {nodes [
  n1 = MaterialFunctionCall() {materialFunction=/BRCosmetics/Athena/Effects/Particles/Skydiving/NS_Contrail_DarkAndLight/Functions/MF_V}
  n2 = MaterialFunctionCall() {materialFunction=/Game/Effects/Fort_Effects/MaterialFunctions/MF_U}
  n3 = ParticleDirection()
  n4 = MaterialFunctionCall() {materialFunction=&CameraDirectionVector}
  n5 = Saturate(None=Multiply(A=DotProduct(A=@n3, B=@n4.Result)) {constB=-1})
  n6 = Multiply(A=@n5, B=@n5)
  n7 = Multiply(A=@n6, B=@n6)
  n8 = Multiply(A=@n7, B=@n7)
  n9 = LinearInterpolate(Alpha=Multiply(A=@n8, B=@n8)) {constA=0.2, constB=0.333}
  n10 = OneMinus(None=Saturate(None=Divide(A=@n2.U, B=@n9)))
  n11 = VectorParameter() {parameterName="Sun Color && Hue Shift", defaultValue=<rgba 9 0.45 0 0.1>}
  n12 = Multiply(A=@n10, B=@n10)
  n13 = Multiply(
   A=Sine(None=Divide(A=@n1.V))
   B=MaterialFunctionCall(Hue Shift Percentage=Multiply(A=@n10, B=@n11.A), Texture=Multiply(A=@n12, B=@n11.RGB)) {materialFunction=&HueShift}.Result
  )
  n14 = OneMinus(None=MaterialFunctionCall() {materialFunction=&MF_U}.U)
  n15 = ScalarParameter() {defaultValue=2, parameterName=Skew, sliderMax=1}
  n16 = TextureCoordinate()
  n17 = ComponentMask(None=@n16) {g=true}
  n18 = ComponentMask(None=@n16) {r=true}
  n19 = OneMinus(None=MaterialFunctionCall() {materialFunction=&MF_U}.U)
  n20 = Multiply(A=@n19, B=@n19)
  n21 = Saturate(None=MaterialFunctionCall!(In=ScalarParameter() {defaultValue=0.3, parameterName=Rainbow Beam Width, sliderMax=1}, Scale=OneMinus(None=@n20)) {materialFunction=/BRCosmetics/Athena/Effects/Particles/Skydiving/NS_Contrail_DarkAndLight/Functions/MF_ValueScaleTo1}.Output)
  n22 = ScalarParameter() {defaultValue=3, parameterName=Rainbow Wave Speed, sliderMax=1}
  n23 = MaterialFunctionCall() {materialFunction=&/MF_DarkAndLight_TimeOffset}
  n24 = MaterialFunctionCall!(
   UV=OneMinus(None=AppendVector(A=@n17, B=@n18))
   U Frequency
   V Frequency=ScalarParameter() {defaultValue=3, parameterName="V Frequency (S)"}
   U Amplitude
   V Amplitude=Multiply(A=OneMinus(None=@n21), B=ScalarParameter() {defaultValue=0.111, desc=was .333, parameterName="V Amplitude (S)"})
   U Speed=@n22
   V Speed=@n22
   U Time=@n23.Time with Offset
   V Time=@n23.Time with Offset
  ) {materialFunction=/Game/Effects/Fort_Effects/MaterialFunctions/MF_WaveUV}
  n25 = Divide(A=ConstantBiasScale(None=ComponentMask(None=@n24.Wave UVs) {r=true}) {bias=-0.5, scale=2}, B=@n21)
  n26 = Saturate(None=DotProduct(A=@n25, B=@n25))
  n27 = OneMinus(None=@n26)
  n28 = Divide(A=@n25, B=Add(A=SquareRoot(None=@n27)))
  n29 = Multiply(
   A=Add(A=Multiply(A=@n14, B=@n15), B=ConstantBiasScale(None=@n28))
   B=ScalarParameter() {defaultValue=0.666, parameterName=Rainbow Color Width, sliderMax=1}
  )
  n30 = Multiply(
   A=MaterialFunctionCall() {materialFunction=&MF_DarkAndLight_TimeOffset}.Time with Offset
   B=ScalarParameter() {defaultValue=1.8, parameterName=Rainbow Spin Rate, sliderMax=1}
  )
  n31 = MaterialFunctionCall!(Gradient=Frac(None=Add(A=@n29, B=@n30))) {materialFunction=&/MF_DarkAndLight_RainbowColors}
  n32 = MaterialFunctionCall() {materialFunction=&MF_U}
  n33 = Subtract(A=Multiply(A=OneMinus(None=@n32.U), B=@n27), B=ComponentMask(None=@n31.Result) {a=true})
  n34 = Multiply(A=ComponentMask(None=@n31.Result) {b=true, g=true, r=true}, B=@n33)
  FunctionOutput(None=AppendVector(A=Add(A=@n13, B=@n34), B=@n33))
  FunctionOutput(None=@n26) {outputName=TransitionGradient}
  ScalarParameter() {defaultValue=0.25, parameterName=Sun Color Length, sliderMax=1}
 ]}
