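# Unlit, alpha-composite Niagara ribbon material for the Dark and Light skydiving contrail.
# The dark look layers inky smoke, triplanar nebula and a hue-shifted eclipse glow that varies with view angle along particle direction, shaped by panning, eroded noise.
# An erosion mask driven by the shared transition function and Reveal Delay blends this into the rainbow ribbon function.
# Depth fade, varying along ribbon U, scales the blended color and alpha.
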Material /BRCosmetics/Athena/Effects/Particles/Skydiving/NS_Contrail_DarkAndLight/M_DarkAndLight_Ribbon
graph {
 properties {
  bAutomaticallySetUsageInEditor=false
  bUsedWithNiagaraRibbons=true
  bUsedWithStaticMesh=false
  blendMode=BLEND_AlphaComposite
  pixelDepthOffsetMode=PDOM_Legacy
  shadingModel=MSM_Unlit
  usedShadingModels=Unlit
 }
 nodes [
  n1 = Divide(A=MaterialFunctionCall() {materialFunction=/Game/Effects/Fort_Effects/MaterialFunctions/MF_U}.U) {constB=0.1}
  n2 = ScalarParameter() {defaultValue=222, parameterName=DepthFadeDistanceA}
  n3 = ScalarParameter() {defaultValue=22, parameterName=DepthFadeDistanceB}
  n4 = Saturate(None=Subtract(A=@n1) {constB=0.25})
  n5 = TextureCoordinate()
  n6 = ComponentMask(None=@n5) {r=true}
  n7 = OneMinus(None=Multiply(A=@n6, B=@n6))
  n8 = TextureCoordinate()
  n9 = ComponentMask(None=@n8) {g=true}
  n10 = Constant() {r=0.5}
  n11 = Abs(None=Divide(A=Subtract(A=@n9, B=@n10), B=@n10))
  n12 = OneMinus(None=@n11)
  n13 = ScalarParameter() {defaultValue=-1, parameterName=Noise Skew, sliderMax=1}
  n14 = AppendVector(A=Add(A=ComponentMask(None=@n5) {g=true}, B=Multiply(A=@n6, B=@n13)), B=@n6)
  n15 = AppendVector(A=ScalarParameter() {defaultValue=0.3, parameterName=Noise Scale U}, B=ScalarParameter() {defaultValue=1.5, parameterName=Noise Scale V})
  n16 = ScalarParameter() {defaultValue=0.07, parameterName=Panner Speed X}
  n17 = ScalarParameter() {defaultValue=-1.66, parameterName=Panner Speed Y}
  n18 = TextureSample(UVs=Panner(
    Coordinate=Multiply(A=@n14, B=@n15)
    Time=MaterialFunctionCall() {materialFunction=&/MF_DarkAndLight_TimeOffset}.Time with Offset
    Speed=AppendVector(A=@n16, B=@n17)
   ) {bFractionalPart=true}) {
   samplerType=SAMPLERTYPE_LinearGrayscale
   texture=&/T_DarkAndLight_Noise
  }
  n19 = SquareRoot(None=@n6)
  n20 = ScalarParameter() {defaultValue=2, parameterName=Noise Modulation, sliderMax=9}
  n21 = OneMinus(None=MaterialFunctionCall!(Gradient=Multiply(A=@n7, B=@n12), Noise=OneMinus(None=@n18.R), Modulation Amount=Multiply(A=@n19, B=@n20)) {materialFunction=/Game/Effects/Fort_Effects/MaterialFunctions/MF_GradientNoiseModulate}.Result)
  n22 = ScalarParameter() {defaultValue=0.5, parameterName=Noise Erosion, sliderMax=1}
  n23 = Saturate(None=OneMinus(None=MaterialFunctionCall!(Gradient=@n21, Erosion Amount=@n22) {materialFunction=/BRCosmetics/Athena/Effects/Particles/Skydiving/NS_Contrail_DarkAndLight/Functions/MF_ErosionNormalized}.Result))
  n24 = ParticleDirection()
  n25 = MaterialFunctionCall() {materialFunction=&CameraDirectionVector}
  n26 = Saturate(None=Multiply(A=DotProduct(A=@n24, B=@n25.Result)) {constB=-1})
  n27 = Multiply(A=@n26, B=@n26)
  n28 = Multiply(A=@n27, B=@n27)
  n29 = Multiply(A=@n28, B=@n28)
  n30 = Multiply(A=@n29, B=@n29)
  n31 = Divide(
   A=MaterialFunctionCall(
    Gradient=MaterialFunctionCall() {materialFunction=&MF_U}.U
    Noise=@n18.R
    Modulation Amount=LinearInterpolate(Alpha=@n30) {constA=0.1, constB=0.3}
   ) {materialFunction=&MF_GradientNoiseModulate}.Result
   B=LinearInterpolate(Alpha=@n30) {constA=0.333, constB=0.5}
  )
  n32 = Power(Base=OneMinus(None=Saturate(None=@n31)), Exp=LinearInterpolate(Alpha=@n30) {constA=6, constB=2})
  n33 = VectorParameter() {parameterName="Eclipse Color && Hue Shift", defaultValue=<rgba 0 1.35 3 -0.1063>}
  n34 = MaterialFunctionCall(Hue Shift Percentage=Multiply(A=@n32, B=@n33.A), Texture=@n33.RGB) {materialFunction=&HueShift}
  n35 = TextureSample(UVs=Panner(Coordinate=TextureCoordinate() {uTiling=0.7, vTiling=0.8}) {speedX=0.2}) {texture=&T_Inky_Smoke_Tile}
  n36 = TextureObject() {texture=/Game/Characters/Player/Male/Medium/Bodies/M_Med_Soldier_04/Skins/Galaxy/Textures/T_FN_Nebula}
  n37 = Constant3Vector() {constant=<rgba 1.5 1.5 1.5 0>}
  n38 = LinearInterpolate(
   A=LinearInterpolate(A=Constant3Vector() {constant=<rgba 0.0015 0.0028 0.0104 1>}, B=Constant3Vector() {constant=<rgba 0.2135 0.0266 0.1457 1>}, Alpha=@n35.R)
   B=Multiply(A=MaterialFunctionCall!(CameraVector, Texture=@n36, Tiling=@n37, Offset, AxisFadeContrast) {materialFunction=/Engine/Functions/Engine_MaterialFunctions01/Texturing/TriplanarCameraVector}.Result) {constB=3}
   Alpha=@n35.R
  )
  n39 = Multiply(A=@n23, B=Add(A=Multiply(A=@n32, B=@n34.Result), B=@n38))
  n40 = MaterialFunctionCall() {materialFunction=&/MF_DarkAndLight_RainbowRibbon}
  n41 = MaterialFunctionCall() {materialFunction=&/MF_DarkAndLight_Transition}
  n42 = ScalarParameter() {parameterName=Reveal Delay, sliderMax=1}
  n43 = Saturate(None=Divide(A=Subtract(A=@n41.Result, B=@n42), B=Subtract(B=@n42)))
  n44 = Power(Base=@n43, Exp=ScalarParameter() {defaultValue=1, parameterName=Exponent, sliderMax=9, sliderMin=1}) {constExponent=3}
  n45 = Divide(A=ScalarParameter() {defaultValue=0.15, parameterName=Softness, sliderMax=1}, B=OneMinus(None=@n41.Result))
  n46 = Multiply(
   A=DepthFade(Opacity=Saturate(None=@n1), FadeDistance=LinearInterpolate(A=@n2, B=@n3, Alpha=@n4))
   B=LinearInterpolate(
    A=AppendVector(A=@n39, B=@n23)
    B=@n40.Result
    Alpha=MaterialFunctionCall!(Gradient=@n40.TransitionGradient, Erosion Amount=@n44, Softness=@n45) {materialFunction=/BRCosmetics/Athena/Effects/Particles/Skydiving/NS_Contrail_DarkAndLight/Functions/MF_Erosion}.Increased Range
   )
  )
  ComponentMask(None=@n46) {a=true}
  Divide(A=@n44) {constB=3}
  n47 = ComponentMask(None=@n8) {r=true}
  n48 = Multiply(A=Saturate(None=Divide(A=@n47) {constB=0.333}), B=@n11)
  Divide(A=@n48) {constB=0.5}
  MaterialFunctionCall(Emissive=ComponentMask(None=@n46) {b=true, g=true, r=true}) {materialFunction=&MF_FX_EmissiveAdjustments}
  n49 = TextureObject() {samplerType=SAMPLERTYPE_Masks, texture=/Game/Characters/Player/Male/Medium/Bodies/M_Med_Soldier_04/Skins/Galaxy/Textures/Celestial_T_Stars}
  n50 = MaterialFunctionCall(Texture=@n49, Tiling=Constant3Vector() {constant=<rgba 3 3 3 0>}, Offset=Multiply(A=Time()) {constB=0.02}) {materialFunction=&TriplanarCameraVector}
  n51 = Add(A=Multiply(A=ComponentMask(None=@n50.Result) {r=true}) {constB=0.85}, B=@n38)
  QualitySwitch(Default=@n51, Low=@n38, High=@n51, Medium=@n38, Epic=@n51)
  ScalarParameter() {defaultValue=0.3, parameterName=Vertical Noise Modulation, sliderMax=1}
  ScalarParameter() {defaultValue=1, parameterName=Vertical Noise Range, sliderMax=1}
  ScalarParameter() {defaultValue=2, parameterName=Vertical Noise Exponent, sliderMax=9, sliderMin=1}
  SquareRoot(None=@n43)
  SquareRoot(None=@n47)
  SquareRoot(None=@n48)
 ]
}
