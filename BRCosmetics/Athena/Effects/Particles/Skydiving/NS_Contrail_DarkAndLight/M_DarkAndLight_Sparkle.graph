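# Unlit alpha-composite Niagara sprite material for sparkles in the DarkAndLight skydiving contrail.
# A rotated rounded-square gradient, masked radially, forms the sparkle and drives hue shifts.
# A transition function blends sun color, shifting to rainbow with age, into noise-modulated eclipse color.
# Lifetime, near-camera and height-above-actor fades shape emissive and opacity, skipping depth fade; a lifetime-and-random offset pulls vertices toward particle center.
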Material /BRCosmetics/Athena/Effects/Particles/Skydiving/NS_Contrail_DarkAndLight/M_DarkAndLight_Sparkle
graph {
 properties {
  bUsedWithNiagaraSprites=true
  blendMode=BLEND_AlphaComposite
  pixelDepthOffsetMode=PDOM_Legacy
  shadingModel=MSM_Unlit
  usedShadingModels=Unlit
 }
 nodes [
  DepthFade(FadeDistance=ParticleRadius())
  n1 = ParticleColor()
  n2 = VectorParameter() {parameterName="Sun Color && Hue Shift", defaultValue=<rgba 3 0.15 0 0.1>}
  n3 = TextureCoordinate()
  n4 = Constant() {r=0.125}
  n5 = Frac(None=Subtract(A=Rotator(Coordinate=@n3, Time=@n4) {speed=6.28}) {constB=0.5})
  n6 = ScalarParameter() {defaultValue=1, parameterName=Size, sliderMax=2}
  n7 = ScalarParameter() {defaultValue=0.3524, parameterName=Curvature, sliderMax=1}
  n8 = Constant() {r=0.5}
  n9 = Divide(A=Distance(A=TextureCoordinate(), B=@n8), B=@n8)
  n10 = Multiply(
   A=OneMinus(None=MaterialFunctionCall!(UV=@n5, Size=@n6, Curvature=@n7) {materialFunction=/Game/Effects/Fort_Effects/MaterialFunctions/MF_RoundedSquare}.Linear Gradient)
   B=OneMinus(None=Saturate(None=@n9))
  )
  n11 = Multiply(A=@n2.A, B=@n10)
  n12 = Time()
  n13 = ParticleRelativeTime()
  n14 = MaterialFunctionCall!(Gradient=Frac(None=Add(A=@n12, B=@n13))) {materialFunction=&/MF_DarkAndLight_RainbowColors}
  n15 = Multiply(A=@n13, B=@n13)
  n16 = VectorParameter() {parameterName="Eclipse Color && Hue Shift", defaultValue=<rgba 0 1.35 3 -0.1063>}
  n17 = Multiply(A=@n16.A, B=@n10)
  n18 = TextureSample(UVs=@n13) {
   samplerType=SAMPLERTYPE_LinearGrayscale
   texture=&/T_DarkAndLight_Noise
  }
  n19 = LinearInterpolate(
   A=LinearInterpolate(
    A=MaterialFunctionCall(Hue Shift Percentage=@n11, Texture=@n2.RGB) {materialFunction=&HueShift}.Result
    B=ComponentMask(None=@n14.Result) {b=true, g=true, r=true}
    Alpha=@n15
   )
   B=Multiply(
    A=MaterialFunctionCall(Hue Shift Percentage=@n17, Texture=@n16.RGB) {materialFunction=&HueShift}.Result
    B=MaterialFunctionCall!(In=@n18.R, Scale=@n15) {materialFunction=/BRCosmetics/Athena/Effects/Particles/Skydiving/NS_Contrail_DarkAndLight/Functions/MF_ValueScaleTo1}.Output
   )
   Alpha=MaterialFunctionCall() {materialFunction=&/MF_DarkAndLight_Transition}.Result
  )
  n20 = Power(Base=@n10, Exp=ScalarParameter() {defaultValue=3.819, parameterName=Exponent, sliderMax=9, sliderMin=1})
  n21 = ParticleRelativeTime()
  n22 = Divide(A=Multiply(A=Multiply(A=@n1.RGB, B=@n19), B=@n20), B=OneMinus(None=@n21))
  n23 = OneMinus(None=MaterialFunctionCall!(Gradient=@n21, Center Point=Constant() {r=0.18}) {materialFunction=/Game/Effects/Fort_Effects/MaterialFunctions/MF_GradientReflected}.Result)
  n24 = ScalarParameter() {defaultValue=333, group="6.1) Opacity", parameterName=Camera Distance Falloff, sortPriority=4}
  n25 = ParticlePositionWS()
  n26 = ActorPositionWS()
  n27 = Subtract(A=ComponentMask(None=Subtract(A=@n25, B=@n26)) {b=true}, B=ScalarParameter() {defaultValue=33, parameterName=Offset})
  n28 = Multiply(A=Multiply(
    A=MaterialFunctionCall(Distance=@n24) {materialFunction=&Near_Camera_Fade}.Result
    B=Saturate(None=@n27)
   )) {constA=1, desc=Skipping depthfade}
  MaterialFunctionCall(Emissive=Multiply(A=Multiply(A=@n22, B=@n23), B=@n28)) {materialFunction=&MF_FX_EmissiveAdjustments}
  Multiply(A=Multiply(A=Multiply(A=@n1.A, B=@n20), B=@n23), B=@n28)
  n29 = Multiply(A=@n23, B=@n23)
  n30 = OneMinus(None=ParticleRandom())
  Multiply(A=Max(A=Multiply(A=@n29, B=@n29), B=Multiply(A=@n30, B=@n30)), B=Subtract(A=ParticlePositionWS(), B=WorldPosition().XYZ))
 ]
}
