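# Skydiving contrail cosmetic "DarkAndLight" that trails effects from two skeletal mesh bones, using spawn groups to offset and vary materials per side.
# Suns burst two persistent camera-offset sprites that follow the bones and take time-of-day color.
# Ribbons leave short trails while Sparkles shed spinning sprites, both scaling velocity with the Dive value.
# Front-end modifiers swap velocities, sparkle size and a material parameter between gameplay and front-end display.
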NiagaraSystem /BRCosmetics/Athena/Effects/Particles/Skydiving/NS_Contrail_DarkAndLight/NS_Contrail_DarkAndLight
dependencies {
 usedDataInterfaces []
 usedDynamicInputs []
 usedModules [
  /Game/Effects/Niagara/Modules/Skydiving/FN_DivingAttributes/v5/FN_DivingAttributes
  &self:SystemSpawnScript.SystemScriptSource.SystemScriptGraph.NiagaraNodeAssignment_1.SetVariables_#1
  &SystemState
 ]
}
emitters {
 Ribbons {
  emitterUpdateScript [
   &EmitterState(Life Cycle Mode=::NewEnumerator1, Loop Duration=0, UseLoopDelay=false)
   /Game/Effects/Niagara/Modules/Skydiving/FN_DivingAttributes/FN_EmitterDivingAttributes_v1(Bone Index=1, SkeletalMesh:NiagaraDataInterfaceSkeletalMesh=$User.SkeletalMesh)
   &FN_SpawnRibbon_v2(Number of Ribbons=2, Spawn Mode=::NewEnumerator0, SpawnRate=64)
  ]
  particleSpawnScript [
   &FN_SetSpawnGroup_v2(Number of Spawn Groups=2)
   &FN_InitializeRibbon_v3(Custom Offset=true, Enable Skeletal Position=true, Lifetime=0.3, Local Vector Offset:Vector3f=<0 1 0>, Ribbon Width=22)
   &AddVelocity(
    Scale Added Velocity=&Lerp_Vector(A=<1 1 1>, B=<4 4 1>, Alpha=$System.Dive)
    Velocity=&FN_FrontEndVector3Modifier(Gameplay=<0 0 999>, Front End=<0 -333 333>)
   )
   &DynamicMaterialParameters(
    Index 0 Param 1=&MaskFloatBySpawnGroup(Masked Float=0.333, Passthrough Float=0.666)
    Index 0 Param 2=&FN_FrontEndFloatModifier(Gameplay=0, Front End=1)
    Module.Param0WriteEnabled=true
    Module.Param1WriteEnabled=true
    Module.Param2WriteEnabled=false
    Module.Param3WriteEnabled=false
   )
   &SkeletalMeshLocation(
    Apply Sampled Orientation to Mesh Particles=true
    Bone / Socket Index=$Particles.SpawnGroup
    Bone Sampling Mode=::NewEnumerator4
    Inherit Velocity=false
    Interpolate Mesh Sampling=false
    MaskSpawnGroup=false
    Sampled Position Offset=&MaskVectorBySpawnGroup(Masked Vector=<10 0 0>, Passthrough Vector=<-10 0 0>)
    Sampled Position Offset Coordinate Space:ENiagaraMeshTransforms=::NewEnumerator3
    Skeletal Mesh=$User.SkeletalMesh
   )
  ]
  particleUpdateScript [
   &InheritSourceMovement(Applied Position Scale=<0.4 0.4 1>, ApplyForce=false, ApplyPosition=true, ApplyPositionFalloff=false)
   &SolveForcesAndVelocity(Clamp Velocity=false, Limit Acceleration=false)
   &FN_UpdateRibbon(Detach Locked Particle=true, Enable Skeletal Position=true)
  ]
  rendererClasses [&NiagaraRibbonRendererProperties]
  renderers [{rendererClass=&NiagaraRibbonRendererProperties}]
 }
 Sparkles {
  emitterSpawnScript [&FN_PreserveSkydiveUserVariables() {module=PreserveUserVariables}]
  emitterUpdateScript [
   &EmitterState()
   &SpawnRate(
    SpawnRate=&FN_FrontEndFloatModifier(
     Gameplay=&Lerp_Float(A=33, B=66, Alpha=$System.Dive)
     Front End=&Lerp_Float(A=33, B=66, Alpha=$System.Dive)
    )
    Use Spawn Probability=false
   )
   /Game/Effects/Niagara/Modules/Skydiving/FN_Emitter_SkydiveAttributes() {module=Emitter_SkydiveAttributes}
  ]
  particleSpawnScript [
   &FN_SetSpawnGroup(Number of Spawn Groups=2)
   &InitializeParticle(
    Lifetime=&Lerp_Float(
     A=&RandomRangeFloat(Minimum=0.3, Maximum=1)
     B=&RandomRangeFloat(Minimum=0.1, Maximum=0.5)
     Alpha=$System.Dive
    )
    Sprite Rotation Mode=::NewEnumerator2
    Sprite Size Mode=::NewEnumerator3
    Uniform Sprite Size=&FN_FrontEndFloatModifier(Gameplay=9, Front End=13)
    UsePositionOffset=false
    Write Material Random=true
   ) {module=InitializeParticle001}
   &SkeletalMeshLocation(
    Apply Sampled Orientation to Mesh Particles=true
    Bone / Socket Index=$Particles.SpawnGroup
    Bone Sampling Mode=::NewEnumerator4
    Inherit Velocity=false
    Interpolate Mesh Sampling=false
    MaskSpawnGroup=false
    Sampled Position Offset=&MakeVector(X=0, Y=0, Z=0)
    Sampled Position Offset Coordinate Space:ENiagaraMeshTransforms=::NewEnumerator2
    Skeletal Mesh=$User.SkeletalMesh
   )
   &AddVelocity(
    Scale Added Velocity=&Lerp_Vector(A=<0.45 0.45 0.6>, B=<2.7 2.7 0.7>, Alpha=$System.Dive)
    Velocity=&FN_FrontEndVector3Modifier(Gameplay=<0 0 999>, Front End=<0 -111 111>)
   )
   &DynamicMaterialParameters(
    Index 0 Param 1=&MaskFloatBySpawnGroup(Masked Float=0.5, Passthrough Float=0)
    Index 0 Param 2=&FN_FrontEndFloatModifier(Gameplay=0, Front End=1)
    Module.Param0WriteEnabled=false
    Module.Param1WriteEnabled=true
    Module.Param2WriteEnabled=false
    Module.Param3WriteEnabled=false
   )
  ]
  particleUpdateScript [
   &ParticleState()
   &InheritSourceMovement(
    Applied Position Scale=&RandomRangeVector(Minimum=<0 0 0.4>, Maximum=<0.4 0.4 1>, Evaluation Type=::NewEnumerator1)
    ApplyForce=false
    ApplyPosition=true
    ApplyPositionFalloff=false
   )
   &SolveForcesAndVelocity(Clamp Velocity=false, Limit Acceleration=false)
   &SpriteRotationRate(Rotation Rate=&RandomRangeFloat(Minimum=-999, Maximum=999))
  ]
  rendererClasses [&NiagaraSpriteRendererProperties]
  renderers [{rendererClass=&NiagaraSpriteRendererProperties}]
 }
 Suns {
  emitterSpawnScript [&FN_PreserveSkydiveUserVariables() {module=PreserveUserVariables}]
  emitterUpdateScript [
   &EmitterState(Inactive Response=::NewEnumerator1, Life Cycle Mode=::NewEnumerator1, Loop Duration=0, UseLoopDelay=false)
   &SpawnBurst_Instantaneous(Spawn Count=2, Spawn Probability=HlslExpression() {hlslExpression="Emitter.LoopCount == 0"}, Use Spawn Probability=true) {module=SpawnBurst_Instantaneous001}
  ]
  particleSpawnScript [
   &FN_SetSpawnGroup_v2(Number of Spawn Groups=2)
   &InitializeParticle(
    Color Mode=::NewEnumerator0
    Lifetime=0
    Sprite Size Mode=::NewEnumerator3
    Uniform Sprite Size=40
    UsePositionOffset=false
    Write Material Random=true
   ) {module=InitializeParticle001}
   &CameraOffset(Camera Offset Amount=22)
   &Color(Color=$NPC.FortniteNPC.FortniteActiveTimeOfDay)
   &DynamicMaterialParameters(
    Index 0 Param 1=&MaskFloatBySpawnGroup(Masked Float=0.5, Passthrough Float=0)
    Index 0 Param 2=&FN_FrontEndFloatModifier(Gameplay=0, Front End=1)
    Module.Param0WriteEnabled=true
    Module.Param1WriteEnabled=true
    Module.Param2WriteEnabled=false
    Module.Param3WriteEnabled=false
   )
  ]
  particleUpdateScript [
   &ParticleState(Kill Particles When Lifetime Has Elapsed=false)
   &SampleSkeletalMeshSkeleton(
    Mesh:NiagaraDataInterfaceSkeletalMesh=$User.SkeletalMesh
    Sampled Rotation Coordinate Space:ENiagaraCoordinateSpace=::Local
    Specific Bone Index:NiagaraInt32=$Particles.SpawnGroup
   )
   &SkeletalMeshLocation(
    Apply Sampled Orientation to Mesh Particles=true
    Bone / Socket Index=$Particles.SpawnGroup
    Bone Sampling Mode=::NewEnumerator4
    Inherit Velocity=false
    Interpolate Mesh Sampling=false
    MaskSpawnGroup=false
    Sampled Position Offset=&MaskVectorBySpawnGroup(Masked Vector=<10 0 0>, Passthrough Vector=<-10 0 0>)
    Sampled Position Offset Coordinate Space:ENiagaraMeshTransforms=::NewEnumerator3
    Skeletal Mesh=$User.SkeletalMesh
   )
  ]
  rendererClasses [&NiagaraSpriteRendererProperties]
  renderers [
   {rendererClass=&NiagaraSpriteRendererProperties}
   {rendererClass=&NiagaraSpriteRendererProperties, rendererIndex=1}
  ]
 }
}
summary {
 systemName=NS_Contrail_DarkAndLight
 emitters [
  {emitterName=Suns, rendererClasses [&NiagaraSpriteRendererProperties]}
  {emitterName=Ribbons, rendererClasses [&NiagaraRibbonRendererProperties]}
  {emitterName=Sparkles, rendererClasses [&NiagaraSpriteRendererProperties]}
 ]
 userVariables [
  User.bIsFrontEnd {type=NiagaraBool}
  User.bIsFrontEndPreview {type=NiagaraBool}
  User.ParaGlideLeanAlpha {type=NiagaraFloat, defaultValue=0, description=""}
  User.PlayerSpeed {type=NiagaraFloat, defaultValue=0, description=""}
  User.RibbonIDOffset {type=NiagaraInt32}
  User.SkeletalMesh {
   type=NiagaraDataInterfaceSkeletalMesh
   defaultValue {
    struct=NiagaraExt_VariableValue_DataInterface
    value {dataInterface=&self:NiagaraDataInterfaceSkeletalMesh_0}
   }
  }
 ]
}
