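# Material function in the DarkAndLight skydiving contrail folder that applies a Photoshop-style Overlay blend to two scalar values, Base and Blend.
# It saturates both inputs, then computes a doubled multiply branch and a doubled screen branch from them.
# A Lerp driven by the rounded Base selects multiply for dark values and screen for light values, avoiding an If node, and outputs the blended float.
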MaterialFunction /BRCosmetics/Athena/Effects/Particles/Skydiving/NS_Contrail_DarkAndLight/Functions/MF_BlendOverlay_Float
graph {
 properties {description=Overlay Blend, libraryCategoriesText [- TuataraVFX]}
 nodes [
  n1 = Saturate(None=FunctionInput(Preview=Constant() {r=0.5}) {inputName=Base, inputType=FunctionInput_Scalar})
  n2 = Multiply(A=@n1) {constB=2, desc=Base}
  n3 = Saturate(None=FunctionInput(Preview=Constant() {r=0.5}) {inputName=Blend, inputType=FunctionInput_Scalar, sortPriority=1})
  n4 = Multiply(A=Multiply(A=OneMinus(None=@n1) {desc=Base}) {constB=2}, B=OneMinus(None=@n3) {desc=Blend})
  FunctionOutput(None=LinearInterpolate(A=Multiply(A=@n2, B=@n3) {desc=Blend}, B=OneMinus(None=@n4), Alpha=Round(None=@n1))) {description=Overlay Blended Float}
 ]
}
