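# Material function in the DarkAndLight skydiving contrail's Functions folder that outputs linear diagonal gradients across UV space.
# It saturates the UV input, which defaults to TextureCoordinate, and dots it with one half to average U and V.
# Flipping U gives the other diagonal, and inverting each result yields four outputs, named by corner value patterns.
# Despite its description of reflection around a specified point, it has no center input.
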MaterialFunction /BRCosmetics/Athena/Effects/Particles/Skydiving/NS_Contrail_DarkAndLight/Functions/MF_GradientDiagonal
graph {
 properties {description=Linear gradient reflected around specified point, libraryCategoriesText [- TuataraVFX]}
 nodes [
  n1 = Saturate(None=FunctionInput(Preview=TextureCoordinate()) {bUsePreviewValueAsDefault=true, inputName=UV, inputType=FunctionInput_Vector2})
  n2 = DotProduct(A=@n1, B=Constant() {r=0.5})
  FunctionOutput(None=@n2) {description=Gradient reflected around centerpoint, outputName="0, 0, 1, 1"}
  FunctionOutput(None=OneMinus(None=@n2)) {description=Gradient reflected around centerpoint, outputName="1, 1, 0, 0", sortPriority=1}
  n3 = MaterialFunctionCall(Float2=@n1) {materialFunction=&BreakOutFloat2Components}
  n4 = DotProduct(A=AppendVector(A=OneMinus(None=@n3.R), B=@n3.G), B=Constant() {r=0.5})
  FunctionOutput(None=@n4) {description=Gradient reflected around centerpoint, outputName="1, 0, 0, 1", sortPriority=2}
  FunctionOutput(None=OneMinus(None=@n4)) {description=Gradient reflected around centerpoint, outputName="0, 1, 1, 0", sortPriority=3}
 ]
}
